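; Quest gameplay ability for the urgent tutorial clamber quest that matches the quest's description to the player's auto-clamber setting.
; On post-initialization it calls UpdateClamberDescOverride, then hits a 5-second Delay noted as a periodic recheck, though no loop back is shown.
; The function runs only off-server, reads the owning pawn's ClamberingComponent, and when the cached setting changes, sets a description override on the transient quests component.
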(event Ability|EventActivateAbility
  (bind _asfort_player_pawn_athena (Utilities|Casting|CastToFortPlayerPawnAthena (Ability|GetAvatarActorfromActorInfo)))
  (Variables|Default|SetOwning%20Player _asfort_player_pawn_athena))

(event Ability|EventFailedToActivatePassiveAbility
  (params (ActorInfo :type Gameplay%20Ability%20Actor%20Info%20Structure))
  (|CallParentFunction))

(event EventOnPostAbilityInitialization
  (CallFunction|UpdateClamberDescOverride)
  (Utilities|FlowControl|Delay 5.0) ; Check occasionally in case the user changes the setting
)

(fn UpdateClamberDescOverride
  (bind _owning_player (Variables|Default|GetOwning%20Player))
  (bind _returnvalue (Actor|GetComponentbyClass _owning_player "/Script/ClamberingCodeRuntime.ClamberingComponent"))
  (bind _returnvalue_2 (Class|ClamberingComponent|IsAutoClamberingEnabled _returnvalue))
  (if (Networking|IsServer)
    _
    (elif (!= (Variables|Default|GetAutoClamberingEnabled) _returnvalue_2)
      (Variables|Default|SetAutoClamberingEnabled _returnvalue_2)
      (bind _asfort_urgent_quest_item_definition (Utilities|Casting|CastToFortUrgentQuestItemDefinition (Variables|FortTransientQuestGameplayAbility|GetQuestItemDefinition)))
      (Class|FortControllerComponentTransientQuests|SetUrgentQuestDescriptionOverride (Class|FortPlayerPawn|GetTransientQuestsComponent _owning_player) (Utilities|Struct|BreakUrgentQuestData (Class|FortUrgentQuestItemDefinition|GetUrgentQuestData _asfort_urgent_quest_item_definition)) (select _returnvalue_2 "(TagName=\"SpecialEvent.Tutorial.02.MoveForward\")" 0)))))
