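; Parent gameplay ability that grants urgent transient tutorial quests in non-competitive Battle Royale playlists.
; After a delay, it scans an objective-to-quest map for an uncompleted objective and grants that quest.
; Completion sends the objective's tagged stat event to advance the upper-level quest, then grants another or ends after the final objective.
; Failure ends the ability; a death/match-end polling timer is defined but never started here.
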(event Ability|EventActivateAbilityFromEvent
  (params (EventData :type Gameplay%20Event%20Data%20Structure%20%28by%20ref%29))
  ;[1 1 1] Get transient quest component from player controller and grant the specified transient quest
  (|CallParentFunction EventData)
  (if (and (Class|FortGameStateAthena|IsBattleRoyalePlaylist (Fortnite|Athena|GameState|GetGameStateAthenafromContext (Variables|GAVars|Data|GetOwningPlayerPawn)) false) (not (Class|FortGameStateAthena|IsCompetitivePlaylist (Fortnite|Athena|GameState|GetGameStateAthenafromContext (Variables|GAVars|Data|GetOwningPlayerPawn)))))
    (bind _asfort_player_controller_athena (Utilities|Casting|CastToFortPlayerControllerAthena (Fortnite|GetFortPlayerControllerfromActor (Variables|GAVars|Data|GetOwningPlayerPawn)))
      (:then
        (Variables|Default|SetFort%20Player%20Controller%20Athena _asfort_player_controller_athena)
        (Utilities|FlowControl|Delay 7.0)
        (CallFunction|SetLastObjectiveName)
        (CallFunction|SetQuesttoGrantbyObjective)
        (Quests|BindEventtoOnUrgentTransientQuestCompleted (Class|FortQuestManager|GetQuestManagerComponent (Variables|GAVars|Data|GetOwningPlayerQuestManager) "/Script/FortniteGame.FortQuestManagerComponent_Athena") (delegate Custom|OnTutorialQuestCompleted))
        ;[1 1 1] If quest fails, see if player is dead and wait for a reboot. Otherwise, end the ability
        (Quests|BindEventto%28DEPRECATED%29OnUrgentTransientQuestFailed (Class|FortQuestManager|GetQuestManagerComponent (Variables|GAVars|Data|GetOwningPlayerQuestManager) "/Script/FortniteGame.FortQuestManagerComponent_Athena") (delegate Custom|OnTransientQuestFailed)))
      (:CastFailed
        _))
    (else
      _))
  (Ability|EndAbility))

(event Custom|OnTransientQuestFailed
  (params (TransientQuestItem :type Fort%20Quest%20Item%20Object%20Reference))
  ;[1 1 1] If quest fails, see if player is dead and wait for a reboot. Otherwise, end the ability
  (if (== (AssetManager|GetPrimaryAssetIdfromObject (Quest|GetQuestDefinition TransientQuestItem)) (AssetManager|GetPrimaryAssetIdfromObject (Variables|Quest|GetTutorialQuest_toGrant)))
    (Ability|EndAbility)))

(event Custom|OnTutorialQuestCompleted
  (params (TransientQuestItem :type Fort%20Quest%20Item%20Object%20Reference))
  ;[1 1 1] Send complex custom stat event when urgent quest is completed to complete upper-level quest
  (if (== (AssetManager|GetPrimaryAssetIdfromObject (Quest|GetQuestDefinition TransientQuestItem)) (AssetManager|GetPrimaryAssetIdfromObject (Variables|Quest|GetTutorialQuest_toGrant)))
    (if (Quest|HasCompletedAllObjectives TransientQuestItem false)
      (Quest|SendComplexCustomStatEvent (Variables|GAVars|Data|GetOwningPlayerQuestManager) 0 (Utilities|Map|Find (Variables|Quest|GetObjectiveToQuestCompleteTag) (Variables|Quest|GetCurrentObjectiveNameToComplete)) "(GameplayTags=)" 1)
      (bind (_playercontrollerout _completedquest) (Quest|HasPlayerCompletedQuest (Fortnite|GetFortPlayerControllerfromActor (Variables|GAVars|Data|GetOwningPlayerPawn)) (Variables|Quest|GetCurrentQuestDefinition) (Variables|Quest|GetFinalQuestObjectiveName))) ; Has player completed the final quest objective?
      (if _completedquest
        (Ability|EndAbility)
        (else
          (Utilities|FlowControl|Delay 2.0)
          (CallFunction|SetQuesttoGrantbyObjective))))))

(event Custom|SetQuestToGrantByObjective
  ;[1 1 1] Find the objective that the player is on and grant the corresponding transient urgent quest
  (Utilities|FlowControl|Delay 4.0)
  (Variables|Quest|SetTutorialQuest_toGrant (Variables|Default|GetNullQuest))
  (bind _keys (Utilities|Map|Keys (Variables|Quest|GetObjectiveToUrgentQuest)))
  (bind _array_element (Utilities|Array|ForEachLoopwithBreak _keys)
    (:LoopBody
      (if (not (Quest|HasCompletedObjectivewithName (Variables|GAVars|Data|GetOwningPlayerQuestManager) (Variables|Quest|GetCurrentQuestDefinition) _array_element :!class /Script/FortniteGame.FortQuestManager))
        (Variables|Quest|SetTutorialQuest_toGrant (Utilities|Map|Find (Variables|Quest|GetObjectiveToUrgentQuest) _array_element))
        (Variables|Quest|SetCurrentObjectiveNameToComplete _array_element)))
    (:Completed
      _))
  (if (!= (Variables|Quest|GetTutorialQuest_toGrant) (Variables|Default|GetNullQuest))
    (Class|FortControllerComponentTransientQuests|GrantTransientQuest (Class|FortPlayerControllerAthena|GetTransientQuestsComponent (Variables|Default|GetFort%20Player%20Controller%20Athena)) (Variables|Quest|GetTutorialQuest_toGrant) 0)
    (else
      (Development|PrintString "No quest found, ending ability" true true "(R=1.000000,G=0.000000,B=0.056822,A=1.000000)" 3.0)
      (Ability|EndAbility))))

(event Custom|SetLastObjectiveName
  ;[1 1 1] Store the name of the last objective that needs to be completed
  (bind _keys (Utilities|Map|Keys (Variables|Quest|GetObjectiveToUrgentQuest)))
  (Variables|Quest|SetFinalQuestObjectiveName (Utilities|Array|Get(aref) _keys (Utilities|Array|LastIndex _keys))))

(event Ability|EventOnEndAbility
  (params (bWasCancelled :type Boolean))
  (bind _owningplayerquestmanager (Variables|GAVars|Data|GetOwningPlayerQuestManager))
  (bind _returnvalue (Class|FortQuestManager|GetQuestManagerComponent _owningplayerquestmanager "/Script/FortniteGame.FortQuestManagerComponent_Athena"))
  ;[1 1 1] On ability ended, clean up event binds
  (Quests|UnbindallEventsfromOnUrgentTransientQuestCompleted _returnvalue)
  (Quests|UnbindallEventsfrom%28DEPRECATED%29OnUrgentTransientQuestFailed _returnvalue))

(event Custom|ShouldContinue
  (bind _returnvalue (Utilities|Time|SetTimerbyEvent (EventDispatchers|CreateEvent :!event Custom|ShouldContinue_Timer) 2.0 true))
  (Variables|Default|SetMatchTimerHandle _returnvalue))

(event Custom|ShouldContinue_Timer
  (Development|PrintString "Should continue calling... " true true "(R=1.000000,G=0.000000,B=0.041007,A=1.000000)")
  (if (FortPawn|IsDead (Variables|GAVars|Data|GetOwningPlayerPawn))
    (if (Game|HasMatchEnded (Fortnite|Athena|GameState|GetGameStateAthenafromContext (Variables|GAVars|Data|GetOwningPlayerPawn)) :!class /Script/Engine.GameStateBase)
      (Development|PrintString "Match ended, ending ability..." true true "(R=1.000000,G=0.000000,B=0.099970,A=1.000000)" 5.0)
      (Utilities|Time|ClearandInvalidateTimerbyHandle (Variables|Default|GetMatchTimerHandle))
      (Ability|EndAbility)
      (else
        (Development|PrintString "Match not over" true true "(R=1.000000,G=0.000000,B=0.099970,A=1.000000)")))
    (else
      (Utilities|Time|ClearandInvalidateTimerbyHandle (Variables|Default|GetMatchTimerHandle))
      (CallFunction|SetQuesttoGrantbyObjective))))
